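# Niagara effect for the Cadet emote that spills tumbling mesh particles ("bullets") which fall, bounce and shrink away.
# Twin bullet emitters share a tapering spawn rate, which FN_FrontEndFloatModifier zeroes for whichever context, front end or gameplay, is inactive.
# The front-end copy collides with an analytical world plane below the owner, standing in for the floor.
# Both generate velocity-thresholded collision events, alongside a Collision_Smoke emitter of rising, fading dust-colored sprites.
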NiagaraSystem /BRCosmetics/Animation/Game/MainPlayer/Emotes/Shorts/VFX/NS_Cadet_Emote
dependencies {usedDataInterfaces [&NiagaraDataInterfaceCurve], usedDynamicInputs [], usedModules [&SystemState]}
emitters {
 Bullets_Frontend {
  emitterUpdateScript [
   &EmitterState()
   &SpawnRate(
    SpawnGroup:NiagaraInt32=&RandomRangeInt(Minimum=0, Maximum=3, Evaluation Type=::NewEnumerator1)
    SpawnRate=&FloatFromCurve(
     FloatCurve=DataInterface() {propertyValues {Curve {keys [{interpMode=RCIM_Cubic, tangentMode=RCTM_User, time=0, value=1}, {interpMode=RCIM_Cubic, tangentMode=RCTM_User, time=1, value=0}]}}}
     CurveIndex=$Emitter.NormalizedLoopAge
     Scale Curve=&FN_FrontEndFloatModifier(Gameplay=0, Front End=44)
    )
    Use Spawn Probability=false
   )
  ]
  particleSpawnScript [
   &InitializeParticle(
    Color Mode=::NewEnumerator0
    Lifetime Max=1.4
    Lifetime Mode=::NewEnumerator1
    Mesh Renderer Array Visibility Mode=::NewEnumerator1
    Mesh Renderer Mesh Index:NiagaraInt32=&MaskIntBySpawnGroup(Masked Int=1, Passthrough Int=0, Spawn Group Mask=0)
    Mesh Scale Mode=::NewEnumerator1
    Mesh Uniform Scale Max=0.9
    Mesh Uniform Scale Min=0.6
    UsePositionOffset=false
    Wrap Index By Mesh Count=false
    Write Material Random=true
   )
   &InitialMeshOrientation(Use Rotation Vector=false)
   &ShapeLocation(MaskSpawnGroup=false, Non Uniform Scale=<1 2 1>, Offset Mode=::NewEnumerator0, Override Seed=false, Sphere Radius=0.1, UseOwnerScale=true)
   &AddVelocity(Velocity=&RandomRangeVector(Minimum=<8 8 0>, Maximum=<-8 -8 0>), Velocity Speed Scale=1.3)
  ]
  particleUpdateScript [
   &ParticleState()
   &GravityForce()
   &Collision(
    Age Colliding Particles=false
    Analytical Collision Plane Position 1:Vector3f=<0 0 0>
    Analytical Collision Plane Position 2:Vector3f=&MakeVector(
     X=0
     Y=0
     Z=&MakeFloatFromVector(
      Vector=&Subtract_Vector(
       A=&ConvertPositionToVector(Input Position=$Engine.Owner.Position, Convert To Absolute Large World Vector=false)
       B=<0 0 133>
      )
      Channel=::NewEnumerator2
     )
    )
    Analytical Collision Plane Space:ENiagaraCoordinateSpace=::World
    CPU Collision Type=::NewEnumerator3
    Control Roll On Collision=true
    Friction=0.35
    Module.Control Rest State Via Surface Normal=true
    Module.EnableMaxCollisionCount=false
    Module.Randomize Collision Normal Vector=true
    Particle Radius Scale=5
    Percentage of Penetration Before Rest=0.5
    Radius Calculation Type=::NewEnumerator1
    Randomize Collision Normal=0.13
    Rest State Time Range=10
    Restitution=0.3
    ScaleRateOfRotation=false
   ) {module=Collision001}
   /Niagara/Modules/Events/GenerateCollisionEvent(Delay Between Collision Events=1.2, Module.EventDelay=true, Module.Use Velocity Threshold=true, Velocity Threshold=100)
   &SolveForcesAndVelocity(Clamp Velocity=false, Limit Acceleration=false)
   &ScaleMeshSize(Scale Factor=&VectorFromFloat(Value=&FloatFromCurve(
      FloatCurve=DataInterface() {propertyValues {Curve {keys [
          <at 0 0.2 RCIM_Cubic>
          {interpMode=RCIM_Cubic, tangentMode=RCTM_User, time=0.3, value=1}
          {arriveTangent=-0.0837, interpMode=RCIM_Cubic, leaveTangent=-0.0837, tangentMode=RCTM_User, time=0.85, value=1}
          <at 1 0>
         ]}}}
      Scale Curve=1.5
     ))) {module=ScaleMeshSize001}
  ]
  rendererClasses [&NiagaraMeshRendererProperties]
  renderers [{rendererClass=&NiagaraMeshRendererProperties, rendererIndex=0}]
 }
 Bullets_Gameplay {
  emitterUpdateScript [
   &EmitterState()
   &SpawnRate(
    SpawnGroup:NiagaraInt32=&RandomRangeInt(Minimum=0, Maximum=3, Evaluation Type=::NewEnumerator1)
    SpawnRate=&FloatFromCurve(
     FloatCurve=DataInterface() {propertyValues {Curve {keys [{interpMode=RCIM_Cubic, tangentMode=RCTM_User, time=0, value=1}, {interpMode=RCIM_Cubic, tangentMode=RCTM_User, time=1, value=0}]}}}
     CurveIndex=$Emitter.NormalizedLoopAge
     Scale Curve=&FN_FrontEndFloatModifier(Gameplay=44, Front End=0)
    )
    Use Spawn Probability=false
   )
  ]
  particleSpawnScript [
   &InitializeParticle(
    Color Mode=::NewEnumerator0
    Lifetime Max=1.4
    Lifetime Mode=::NewEnumerator1
    Mesh Renderer Array Visibility Mode=::NewEnumerator1
    Mesh Renderer Mesh Index:NiagaraInt32=&MaskIntBySpawnGroup(Masked Int=1, Passthrough Int=0, Spawn Group Mask=0)
    Mesh Scale Mode=::NewEnumerator1
    Mesh Uniform Scale Max=0.9
    Mesh Uniform Scale Min=0.6
    UsePositionOffset=false
    Wrap Index By Mesh Count=false
    Write Material Random=true
   )
   &InitialMeshOrientation(Use Rotation Vector=false)
   &ShapeLocation(MaskSpawnGroup=false, Non Uniform Scale=<1 2 1>, Offset Mode=::NewEnumerator0, Override Seed=false, Sphere Radius=0.1, UseOwnerScale=true)
   &AddVelocity(Velocity=&RandomRangeVector(Minimum=<8 8 0>, Maximum=<-8 -8 0>), Velocity Speed Scale=1.1)
  ]
  particleUpdateScript [
   &ParticleState()
   &GravityForce()
   &Collision(
    Age Colliding Particles=false
    Control Roll On Collision=true
    Friction=0.35
    Module.Control Rest State Via Surface Normal=true
    Module.EnableMaxCollisionCount=false
    Module.Randomize Collision Normal Vector=true
    Particle Radius Scale=5
    Percentage of Penetration Before Rest=0.5
    Radius Calculation Type=::NewEnumerator1
    Rest State Time Range=10
    Restitution=0.3
    ScaleRateOfRotation=false
   )
   &GenerateCollisionEvent(Delay Between Collision Events=1.2, Module.EventDelay=true, Module.Use Velocity Threshold=true, Velocity Threshold=100)
   &SolveForcesAndVelocity(Clamp Velocity=false, Limit Acceleration=false)
   &ScaleMeshSize(Scale Factor=&VectorFromFloat(Value=&FloatFromCurve(
      FloatCurve=DataInterface() {propertyValues {Curve {keys [
          <at 0 0.2 RCIM_Cubic>
          {interpMode=RCIM_Cubic, tangentMode=RCTM_User, time=0.3, value=1}
          {arriveTangent=-0.0837, interpMode=RCIM_Cubic, leaveTangent=-0.0837, tangentMode=RCTM_User, time=0.85, value=1}
          <at 1 0>
         ]}}}
      Scale Curve=1.5
     )))
  ]
  rendererClasses [&NiagaraMeshRendererProperties]
  renderers [{rendererClass=&NiagaraMeshRendererProperties, rendererIndex=0}]
 }
 Collision_Smoke {
  emitterUpdateScript [&EmitterState()]
  particleSpawnScript [
   &InitializeParticle(
    Color=&RandomRangeLinearColor(Minimum=<rgba 0.4115 0.3286 0.2562 0.5>, Maximum=<rgba 0.2083 0.1441 0.0948 0.5>)
    Lifetime Max=0.7
    Lifetime Min=1.4
    Lifetime Mode=::NewEnumerator1
    Sprite Rotation Mode=::NewEnumerator2
    Sprite Size Mode=::NewEnumerator1
    Uniform Sprite Size Max=4
    Uniform Sprite Size Min=2
    UsePositionOffset=false
    Write Material Random=true
   ) {module=InitializeParticle002}
   &AddVelocityFromPoint(Module.Use Falloff Distance=false, Velocity Strength=&RandomRangeFloat(Minimum=20, Maximum=40)) {module=AddVelocityFromPoint001}
  ]
  particleUpdateScript [
   &ParticleState()
   &GravityForce(Gravity=<0 0 30>)
   &Drag(Drag=2, Use Linear Drag=true, Use Rotational Drag=false)
   &SolveForcesAndVelocity(Clamp Velocity=false, Limit Acceleration=false)
   &ScaleSpriteSize(Scale Factor=&Vector2DFromFloat(Value=&FloatFromCurve(FloatCurve=DataInterface() {propertyValues {Curve {keys [
          {arriveTangent=87.4549, interpMode=RCIM_Cubic, leaveTangent=87.4549, tangentMode=RCTM_User, time=0, value=1}
          {arriveTangent=4.8897, interpMode=RCIM_Cubic, leaveTangent=4.8897, tangentMode=RCTM_User, time=0.1, value=8}
          {interpMode=RCIM_Cubic, leaveTangent=-6.7553, time=1.0011, value=10}
         ]}}})))
   &ScaleColor(
    Scale Alpha=&FloatFromCurve(
     FloatCurve=DataInterface() {propertyValues {Curve {keys [
         {arriveTangent=17.5364, interpMode=RCIM_Cubic, leaveTangent=17.5364, tangentMode=RCTM_User, time=0, value=0}
         {arriveTangent=-3.1757, interpMode=RCIM_Cubic, tangentMode=RCTM_Break, time=0.1, value=1}
         {arriveTangent=0.0071, interpMode=RCIM_Cubic, leaveTangent=0.0071, tangentMode=RCTM_User, time=0.4, value=1}
         {arriveTangent=-1.6667, interpMode=RCIM_Cubic, leaveTangent=-1.6667, time=0.7, value=0.25}
         <at 1 0 RCIM_Cubic>
        ]}}}
     Scale Curve=0.2
    )
    ScaleA=true
    ScaleRGB=true
   ) {module=ScaleColor001}
   &DynamicMaterialParameters(
    Index 0 Param 1=&Add_Float(
     A=&FloatFromCurve(FloatCurve=DataInterface() {propertyValues {Curve {keys [{arriveTangent=2.1549, time=0, value=0.8}, <at 1 1>]}}})
     B=&Multiply_Float(A=$Particles.Initial.MaterialRandom, B=0.5)
    )
    Index 0 Param 3=&FloatFromCurve(
     FloatCurve=DataInterface() {propertyValues {Curve {keys [{arriveTangent=-0.0944, interpMode=RCIM_Cubic, time=0, value=0}, <at 0.5008 0>]}}}
     Scale Curve=3
    )
    Index 0 Param 4=1
    Index 1 Param 1=1
    Index 1 Param 2=1
    Index 1 Param 3=1
    Index 1 Param 4=1
    Index 2 Param 1=&FloatFromCurve(FloatCurve=DataInterface() {propertyValues {Curve {keys [
         {arriveTangent=2.7333, interpMode=RCIM_Cubic, leaveTangent=2.7333, tangentMode=RCTM_User, time=0, value=0}
         {arriveTangent=-0.0004, interpMode=RCIM_Cubic, leaveTangent=-0.0004, tangentMode=RCTM_User, time=1, value=1}
        ]}}})
    Index 2 Param 2=&RandomRangeFloat(Minimum=0.5, Maximum=1)
    Index 2 Param 3=&FloatFromCurve(FloatCurve=DataInterface() {propertyValues {Curve {keys [
         <at 0 10>
         {arriveTangent=5.218, interpMode=RCIM_Cubic, leaveTangent=5.218, tangentMode=RCTM_User, time=0.1, value=0}
         {arriveTangent=20.2637, interpMode=RCIM_Cubic, leaveTangent=20.2637, tangentMode=RCTM_User, time=1, value=10}
        ]}}})
    Index 2 Param 4=&RandomRangeFloat(Minimum=0.8, Maximum=1.2)
    Module.Param0WriteEnabled=true
    Module.Param0WriteEnabled1=true
    Module.Param0WriteEnabled2=true
    Module.Param0WriteEnabled3=true
    Module.Param1WriteEnabled=true
    Module.Param1WriteEnabled1=true
    Module.Param1WriteEnabled2=true
    Module.Param1WriteEnabled3=true
    Module.Param2WriteEnabled=true
    Module.Param2WriteEnabled1=true
    Module.Param2WriteEnabled2=true
    Module.Param2WriteEnabled3=true
    Module.Param3WriteEnabled=true
    Module.Param3WriteEnabled1=true
    Module.Param3WriteEnabled2=true
    Module.Param3WriteEnabled3=true
    Write Parameter Index 1=true
    Write Parameter Index 2=true
    Write Parameter Index 3=true
   ) {module=DynamicMaterialParameters002}
   &CameraOffset(Camera Offset Amount=50)
   &SpriteRotationRate(Rotation Rate=&RandomRangeFloat(Minimum=-90, Maximum=90))
  ]
  rendererClasses [&NiagaraSpriteRendererProperties]
  renderers [{rendererClass=&NiagaraSpriteRendererProperties}]
 }
}
summary {
 systemName=NS_Cadet_Emote
 emitters [
  {emitterName=Collision_Smoke, rendererClasses [&NiagaraSpriteRendererProperties]}
  {emitterName=Bullets_Gameplay, rendererClasses [&NiagaraMeshRendererProperties]}
  {emitterName=Bullets_Frontend, rendererClasses [&NiagaraMeshRendererProperties]}
 ]
 userVariables [
  User.bIsFrontEnd {type=NiagaraBool}
  User.bIsFrontEndPreview {type=NiagaraBool}
 ]
}
